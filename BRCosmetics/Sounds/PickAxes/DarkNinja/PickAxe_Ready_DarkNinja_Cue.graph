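# Sound cue for the DarkNinja pickaxe ready sound in Battle Royale cosmetics, built from the pickaxe's swing waves.
# An equally weighted random node picks one of three swing waves and feeds a modulator at the root.
# The cue plays them quieter and pitched up, using close projectile attenuation and the melee weapon sound class.
# It primes on load and overrides concurrency to block new instances at the limit.
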SoundCue /BRCosmetics/Sounds/PickAxes/DarkNinja/PickAxe_Ready_DarkNinja_Cue
root=Modulator_1
nodes [
 {id=Modulator_1, children [Random_0]}
 {id=Random_0, children [WavePlayer_0,WavePlayer_1,WavePlayer_2], properties {weights [1,1,1]}}
 {id=WavePlayer_0, properties {soundWaveAssetPtr=&/PA_DarkNinja_Swing_01}}
 {id=WavePlayer_1, properties {soundWaveAssetPtr=&/PA_DarkNinja_Swing_02}}
 {id=WavePlayer_2, properties {soundWaveAssetPtr=&/PA_DarkNinja_Swing_03}}
]
properties {
 attenuationSettings=&Fort_Weapon_Projectile_Close_Athena_Att
 audioPropertiesSheet=/BRCosmetics/Sounds/Settings/PropertySheets/Pickaxe/APS_Cosmetic_Pickaxe_Ready
 bOverrideConcurrency=true
 bPrimeOnLoad=true
 maxDistance=4950
 pitchMultiplier=1.2
 soundClassObject=&Fort_WeaponMelee
 volumeMultiplier=0.15
 concurrencyOverrides {resolutionRule=PreventNew}
}
